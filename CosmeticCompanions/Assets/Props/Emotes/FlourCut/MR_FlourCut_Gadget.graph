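# Modular Control Rig for the musical keyboard prop in the FlourCut emote gadget.
# A root and keyboard hierarchy carries individually named piano-key elements from F3 to Bb4, each with a child socket at its origin.
# Connectors indicate a Root module plus one module per key, each exposing primary and parent-control attachment points.
# The rig's own graph holds only an empty begin-execution node, so any control behavior comes from those modules.
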ControlRigBlueprint /CosmeticCompanions/Assets/Props/Emotes/FlourCut/MR_FlourCut_Gadget
graphs [
 RigVMFunctionLibrary {links [], nodes []}
 RigVMModel {nodes [RigUnit_BeginExecution = &RigUnit_BeginExecution()]}
]
hierarchy [
 root {loc=<0 0 0>, rot=<0 0 0 1>}
 keyboard {parent=root, loc=<0 0 0>, rot=<0 0 0 1>}
 key_Bb3 {parent=keyboard, loc=<-21.0669 9.0621 9.5072>, rot=<0 0 0 1>}
 key_Bb3_socket {parent=key_Bb3, type=Socket, loc=<0 0 0>, rot=<0 0 0 1>}
 key_Db3 {parent=keyboard, loc=<-4.0669 9.0621 9.5072>, rot=<0 0 0 1>}
 key_Db3_socket {parent=key_Db3, type=Socket, loc=<0 0 0>, rot=<0 0 0 1>}
 key_Eb4 {parent=keyboard, loc=<4.5331 9.0621 9.5072>, rot=<0 0 0 1>}
 key_Eb4_socket {parent=key_Eb4, type=Socket, loc=<0 0 0>, rot=<0 0 0 1>}
 key_Gb4 {parent=keyboard, loc=<21.6531 9.0621 9.5072>, rot=<0 0 0 1>}
 key_Gb4_socket {parent=key_Gb4, type=Socket, loc=<0 0 0>, rot=<0 0 0 1>}
 key_Ab4 {parent=keyboard, loc=<30.133 9.0621 9.5072>, rot=<0 0 0 1>}
 key_Ab4_socket {parent=key_Ab4, type=Socket, loc=<0 0 0>, rot=<0 0 0 1>}
 key_Bb4 {parent=keyboard, loc=<38.733 9.0621 9.5072>, rot=<0 0 0 1>}
 key_Bb4_socket {parent=key_Bb4, type=Socket, loc=<0 0 0>, rot=<0 0 0 1>}
 key_Ab3 {parent=keyboard, loc=<-29.7069 9.0621 9.5072>, rot=<0 0 0 1>}
 key_Ab3_socket {parent=key_Ab3, type=Socket, loc=<0 0 0>, rot=<0 0 0 1>}
 key_A4 {parent=keyboard, loc=<38.6346 16.8929 6.6272>, rot=<0 0 0 1>}
 key_A4_socket {parent=key_A4, type=Socket, loc=<0 0 0>, rot=<0 0 0 1>}
 key_F4 {parent=keyboard, loc=<21.5986 16.8929 6.6272>, rot=<0 0 0 1>}
 key_F4_socket {parent=key_F4, type=Socket, loc=<0 0 0>, rot=<0 0 0 1>}
 key_D4 {parent=keyboard, loc=<4.3345 16.8929 6.6272>, rot=<0 0 0 1>}
 key_D4_socket {parent=key_D4, type=Socket, loc=<0 0 0>, rot=<0 0 0 1>}
 key_C4 {parent=keyboard, loc=<-8.3368 16.8929 6.6272>, rot=<0 0 0 1>}
 key_C4_socket {parent=key_C4, type=Socket, loc=<0 0 0>, rot=<0 0 0 1>}
 key_A3 {parent=keyboard, loc=<-21.2771 16.8929 6.6272>, rot=<0 0 0 1>}
 key_A3_socket {parent=key_A3, type=Socket, loc=<0 0 0>, rot=<0 0 0 1>}
 key_G3 {parent=keyboard, loc=<-33.8933 16.8929 6.6272>, rot=<0 0 0 1>}
 key_G3_socket {parent=key_G3, type=Socket, loc=<0 0 0>, rot=<0 0 0 1>}
 key_Gb3 {parent=keyboard, loc=<-38.267 9.0621 9.5072>, rot=<0 0 0 1>}
 key_Gb3_socket {parent=key_Gb3, type=Socket, loc=<0 0 0>, rot=<0 0 0 1>}
 key_F3 {parent=keyboard, loc=<-42.914 16.8929 6.6272>, rot=<0 0 0 1>}
 key_F3_socket {parent=key_F3, type=Socket, loc=<0 0 0>, rot=<0 0 0 1>}
 Root/Root {type=Connector, loc=<0 0 0>, rot=<0 0 0 1>}
 Key Bb3/Add Control Primary {type=Connector, loc=<0 0 0>, rot=<0 0 0 1>}
 Key Bb3/Parent Control {type=Connector, loc=<0 0 0>, rot=<0 0 0 1>}
 Key C4/Add Control Primary {type=Connector, loc=<0 0 0>, rot=<0 0 0 1>}
 Key C4/Parent Control {type=Connector, loc=<0 0 0>, rot=<0 0 0 1>}
 Key Eb4/Add Control Primary {type=Connector, loc=<0 0 0>, rot=<0 0 0 1>}
 Key Eb4/Parent Control {type=Connector, loc=<0 0 0>, rot=<0 0 0 1>}
 Key Gb4/Add Control Primary {type=Connector, loc=<0 0 0>, rot=<0 0 0 1>}
 Key Gb4/Parent Control {type=Connector, loc=<0 0 0>, rot=<0 0 0 1>}
 Key Db3/Add Control Primary {type=Connector, loc=<0 0 0>, rot=<0 0 0 1>}
 Key Db3/Parent Control {type=Connector, loc=<0 0 0>, rot=<0 0 0 1>}
 Key Ab4/Add Control Primary {type=Connector, loc=<0 0 0>, rot=<0 0 0 1>}
 Key Ab4/Parent Control {type=Connector, loc=<0 0 0>, rot=<0 0 0 1>}
 Key Bb4/Add Control Primary {type=Connector, loc=<0 0 0>, rot=<0 0 0 1>}
 Key Bb4/Parent Control {type=Connector, loc=<0 0 0>, rot=<0 0 0 1>}
 Key Ab3/Add Control Primary {type=Connector, loc=<0 0 0>, rot=<0 0 0 1>}
 Key Ab3/Parent Control {type=Connector, loc=<0 0 0>, rot=<0 0 0 1>}
 Key A4/Add Control Primary {type=Connector, loc=<0 0 0>, rot=<0 0 0 1>}
 Key A4/Parent Control {type=Connector, loc=<0 0 0>, rot=<0 0 0 1>}
 Key D4/Add Control Primary {type=Connector, loc=<0 0 0>, rot=<0 0 0 1>}
 Key D4/Parent Control {type=Connector, loc=<0 0 0>, rot=<0 0 0 1>}
 Key A3/Add Control Primary {type=Connector, loc=<0 0 0>, rot=<0 0 0 1>}
 Key A3/Parent Control {type=Connector, loc=<0 0 0>, rot=<0 0 0 1>}
 Key G3/Add Control Primary {type=Connector, loc=<0 0 0>, rot=<0 0 0 1>}
 Key G3/Parent Control {type=Connector, loc=<0 0 0>, rot=<0 0 0 1>}
 Key Gb3/Add Control Primary {type=Connector, loc=<0 0 0>, rot=<0 0 0 1>}
 Key Gb3/Parent Control {type=Connector, loc=<0 0 0>, rot=<0 0 0 1>}
 Key F3/Add Control Primary {type=Connector, loc=<0 0 0>, rot=<0 0 0 1>}
 Key F3/Parent Control {type=Connector, loc=<0 0 0>, rot=<0 0 0 1>}
 Key F4/Add Control Primary {type=Connector, loc=<0 0 0>, rot=<0 0 0 1>}
 Key F4/Parent Control {type=Connector, loc=<0 0 0>, rot=<0 0 0 1>}
]
