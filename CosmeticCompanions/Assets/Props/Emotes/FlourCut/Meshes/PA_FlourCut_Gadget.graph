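# Physics asset for the FlourCut gadget mesh, an emote prop under CosmeticCompanions.
# It holds a single body on the keyboard bone, shaped as one wide capsule rotated and slightly offset from the bone.
# There are no constraints, so the asset provides one simple collision volume with no jointed parts.
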PhysicsAsset /CosmeticCompanions/Assets/Props/Emotes/FlourCut/Meshes/PA_FlourCut_Gadget
bodies {keyboard [{
   shapeType=Capsule
   length=59.1374
   radius=51.1424
   shapeName=keyboard_capsule
   center=<0.0292 0.4226 9.4353>
   rotation=<rot 0 91.6165 -89.9162>
  }]}
constraints []
